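# Idle Niagara effect for the Rose Jewel backpack that sheds small dark-red petal sprites from sockets.
# Spawn rate rises when the owner moves fast; petals briefly follow the source, then drift under curl noise, gravity and drag.
# Sprites borrow mesh orientation to tumble at a velocity-driven rate, with a view-dependent camera offset.
# Size varies with age, camera distance and travel distance; petals die past a random maximum distance.
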NiagaraSystem /BRCosmetics/Accessories/FORT_Backpacks/Backpack_F_MED_RoseJewel/FX/NS_Backpack_RoseJewel_Idle
dependencies {
 usedDataInterfaces [&NiagaraDataInterfaceCurve, &NiagaraDataInterfaceCamera, &NiagaraDataInterfaceSocketReader]
 usedDynamicInputs []
 usedModules [
  &self:SystemSpawnScript.SystemScriptSource.SystemScriptGraph.NiagaraNodeAssignment_1.SetVariables_#1
  &SystemState
 ]
}
emitters {Petals {
  emitterUpdateScript [
   &EmitterState()
   &SpawnRate(
    Spawn Probability=0.8
    SpawnRate=&Multiply_Float(
     A=2.5
     B=&MakeCustomFloatFromBool(
      Bool=&SetBoolByFloatComparison(A=&VectorLength(Vector=$Engine.Owner.Velocity), B=400)
      True Float=2.5
      False Float=1
     )
    )
    Use Spawn Probability=true
   )
  ]
  particleSpawnScript [
   &InitializeParticle(
    Color=<rgba 0.27 0.0023 0.0023 1>
    Lifetime=&RandomRangeFloat(Minimum=1.9, Maximum=2.4)
    Sprite Rotation Mode=::NewEnumerator2
    Sprite Size Max=<9 9>
    Sprite Size Min=<4.8 4.8>
    Sprite Size Mode=::NewEnumerator2
    UsePositionOffset=false
    Write Material Random=true
   )
   &PartitionParticles(Number Of Partitions=10, Partition Mode:ENiagara_PartitionMode_v2=::NewEnumerator5)
   &SocketLocation(
    Interpolate Socket Sampling=true
    Orientation Sampling=::NewEnumerator1
    OverrideRandomSeed=false
    SamplePosOffset=false
    Socket Reader=$System.Socket Reader
    Socket Sampling Mode=::NewEnumerator1
    VelocitySpeedLimit=false
   )
   &AddVelocity(
    Cone Axis=<0 -1 -0.3>
    Module.UseConeFalloff=false
    Module.UseVelDistribution=false
    Override Seed=false
    Velocity Mode=::NewEnumerator2
    Velocity Speed=&RandomRangeFloat(Minimum=22.5, Maximum=60)
   )
   &InitialMeshOrientation(Use Rotation Vector=false)
  ]
  particleUpdateScript [
   &ParticleState()
   /Niagara/Modules/Update/Forces/CurlNoiseForce(
    Noise Frequency=100
    Noise Strength=&Multiply_Float(
     A=100
     B=&FloatFromCurve(FloatCurve=DataInterface() {propertyValues {Curve {keys [{interpMode=RCIM_Cubic, tangentMode=RCTM_User, time=0, value=0}, {interpMode=RCIM_Cubic, tangentMode=RCTM_User, time=1, value=1}]}}})
    )
    Pan Noise=false
   )
   &GravityForce(Gravity=&RandomRangeVector(Minimum=<0 0 -200>, Maximum=<0 0 -350>))
   &AerodynamicDrag(Aerodynamic Drag=0.5, Aerodynamic Rotational Drag=0.5, Use Linear Drag=true, Use Rotational Drag=true)
   &InheritSourceMovement(
    Applied Position Scale=&VectorFromFloat(Value=&FloatFromCurve(FloatCurve=DataInterface() {propertyValues {Curve {keys [
          <at 0.0426 1 RCIM_Cubic>
          {
           arriveTangent=-2.7049
           interpMode=RCIM_Cubic
           leaveTangent=-2.7049
           leaveTangentWeight=0.2482
           tangentMode=RCTM_User
           time=0.1911
           value=0.5292
          }
          {interpMode=RCIM_Cubic, leaveTangent=-0.5401, time=0.7, value=0}
         ]}}}))
    ApplyForce=false
    ApplyPosition=true
    ApplyPositionFalloff=false
   )
   &SolveForcesAndVelocity(Clamp Velocity=false, Limit Acceleration=false)
   &UpdateMeshOrientation(
    Rotation Coordinate Space=::NewEnumerator2
    Rotation Rate=&Divide_Float(A=&VectorLength(Vector=$Particles.Velocity), B=30)
    Rotation Vector=&RandomVector(Vector Scale=1)
   )
   &AlignSpriteToMeshOrientation(SetAlignment=true, SetFacing=true)
   &CameraOffset(Camera Offset Amount=&FloatFromCurve(
     FloatCurve=DataInterface() {propertyValues {Curve {keys [<at 0.0489 0.3>,<at 0.1905 1>]}}}
     Scale Curve=&Multiply_Float(
      A=25
      B=&ClampFloat(Float=&DotProduct(Vector A=&ReturnCameraVector(Camera Vectors=::NewEnumerator1, Coordinate Space=::Local), Vector B=<0 1 0>))
     )
    ))
   &ScaleSpriteSize(
    Uniform Curve Scale=0.8
    Uniform Curve Sprite Scale=DataInterface() {propertyValues {Curve {keys [
        {arriveTangent=7.403, interpMode=RCIM_Cubic, leaveTangent=7.403, tangentMode=RCTM_User, time=0, value=0.4705}
        {arriveTangent=-0.1014, interpMode=RCIM_Cubic, leaveTangent=-0.1014, tangentMode=RCTM_User, time=0.1473, value=1}
        {arriveTangent=-1.2169, interpMode=RCIM_Cubic, leaveTangent=-1.2169, tangentMode=RCTM_User, time=0.727, value=0.7725}
        <at 1 0.25>
       ]}}}
   )
   &ScaleSpriteSize(
    Initial Sprite Size:Vector2f=$Particles.SpriteSize
    Uniform Curve Index=/Game/Effects/Niagara/Modules/Math/FN_CameraDistance(
     0Dist:NiagaraFloat=70
     1Dist:NiagaraFloat=90
     Camera Query:NiagaraDataInterfaceCamera=DataInterface()
     Position:NiagaraPosition=$Particles.Position
    )
   ) {module=ScaleSpriteSize001}
   /SetVariables(
    Particles.MaxDistance:NiagaraFloat=&RandomRangeFloat(Minimum=140, Maximum=170)
    Particles.ParticleDistance:NiagaraFloat=&DistanceBetweenPositions(A=$Engine.Emitter.SimulationPosition, B=$Particles.Position)
   ) {isSetParameters=true}
   &ScaleSpriteSize(
    Uniform Curve Index=&ClampFloat(Float=&Divide_Float(A=$Particles.ParticleDistance, B=$Particles.MaxDistance))
    Uniform Curve Sprite Scale=DataInterface() {propertyValues {Curve {keys [<at 0.43 1>,<at 1 0>]}}}
   ) {module=ScaleSpriteSize002}
   &KillParticles(Kill Particles=&SetBoolByFloatComparison(A=$Particles.ParticleDistance, B=$Particles.MaxDistance))
  ]
  rendererClasses [&NiagaraSpriteRendererProperties]
  renderers [{rendererClass=&NiagaraSpriteRendererProperties}]
 }}
summary {
 systemName=NS_Backpack_RoseJewel_Idle
 emitters [{emitterName=Petals, rendererClasses [&NiagaraSpriteRendererProperties]}]
 userVariables [
  User.bIsFrontEnd {type=NiagaraBool}
  User.bIsFrontEndPreview {type=NiagaraBool}
  User.bPawnSprintOrSlide {type=NiagaraBool}
  User.LandBurst {type=NiagaraFloat, defaultValue=0, description=""}
  User.LandGameTime {type=NiagaraFloat, defaultValue=0, description=""}
  User.ReactivityDebug {type=NiagaraBool}
 ]
}
