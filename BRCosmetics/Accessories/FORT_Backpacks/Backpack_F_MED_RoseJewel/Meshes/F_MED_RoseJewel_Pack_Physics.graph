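# Physics asset for the RoseJewel backpack mesh, setting up collision and joints for a dangling strap on the pack.
# A five-segment capsule chain hangs from a base strap body that is angularly locked to the pack, with swing limit angles widening toward the tip.
# The pack uses two large capsules, while unconstrained pelvis and hip capsules provide body collision volumes around the wearer.
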PhysicsAsset /BRCosmetics/Accessories/FORT_Backpacks/Backpack_F_MED_RoseJewel/Meshes/F_MED_RoseJewel_Pack_Physics
bodies {
 Hip_Collision [{shapeType=Capsule, length=18.8142, radius=14.0284, shapeName=Hip_Collision_capsule, center=<-3.8421 -2.2495 0>, rotation=<rot 0 -90 0>}]
 base_strap [{shapeType=Capsule, length=0.505, radius=0.505, shapeName=base_strap_capsule, rotation=<rot 0 -90 0>}]
 dyn_strap_1 [{shapeType=Capsule, length=2.5, radius=0.7, shapeName=dyn_strap_1_capsule, center=<1.5423 0 0>, rotation=<rot 0 90 -90>}]
 dyn_strap_2 [{shapeType=Capsule, length=3, radius=1.2, shapeName=dyn_strap_2_capsule, center=<2 0 0>, rotation=<rot 0 90 -89.8384>}]
 dyn_strap_3 [{shapeType=Capsule, length=3, radius=1.2, shapeName=dyn_strap_3_capsule, center=<2 0 0>, rotation=<rot 0 90 -89.8384>}]
 dyn_strap_4 [{shapeType=Capsule, length=3, radius=1.2, shapeName=dyn_strap_4_capsule, center=<2 0 0>, rotation=<rot 0 90 -89.8384>}]
 dyn_strap_5 [{shapeType=Capsule, length=7, radius=2, shapeName=dyn_strap_5_capsule, center=<6 0 0>, rotation=<rot 0 90 -89.8384>}]
 pack [
  {shapeType=Capsule, length=46.4697, radius=20, shapeName=pack_capsule, center=<0 19.3865 -26.2767>, rotation=<rot 0 0 0.6879>}
  {
   shapeType=Capsule
   length=16.4154
   radius=37.9329
   shapeName=pack_capsule
   center=<0.6408 28.1163 -66.7767>
   rotation=<rot 81.0327 -179.9997 -179.9997>
  }
 ]
 pelvis [{shapeType=Capsule, length=14.9888, radius=15.7558, shapeName=pelvis_capsule, center=<-4 1.7 0>, rotation=<rot 0 -90 0>}]
}
constraints [
 {
  bone1Name=base_strap
  bone2Name=pack
  swing1LimitDegrees=45
  swing1Motion=Locked
  swing2LimitDegrees=45
  swing2Motion=Locked
  twistLimitDegrees=45
  twistMotion=Locked
 }
 {bone1Name=dyn_strap_1, bone2Name=base_strap, swing1LimitDegrees=15, swing2LimitDegrees=15, twistLimitDegrees=0}
 {bone1Name=dyn_strap_2, bone2Name=dyn_strap_1, swing1LimitDegrees=30, swing2LimitDegrees=30, twistLimitDegrees=0}
 {bone1Name=dyn_strap_3, bone2Name=dyn_strap_2, swing1LimitDegrees=30, swing2LimitDegrees=30, twistLimitDegrees=0}
 {bone1Name=dyn_strap_4, bone2Name=dyn_strap_3, swing1LimitDegrees=30, swing2LimitDegrees=30, twistLimitDegrees=0}
 {bone1Name=dyn_strap_5, bone2Name=dyn_strap_4, swing1LimitDegrees=45, swing2LimitDegrees=45, twistLimitDegrees=0}
]
