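# Anim Blueprint for the RoseJewel backpack that adds secondary motion to its rose and strap bones.
# Adjusted input pose runs through Anim Dynamics on the rose bones, then gets cached.
# bRigidBodyEnabled selects between RigidBody simulation and an Anim Dynamics/Trail strap chain with spherical limits on Hip_Collision.
# The event graph reads parent-mesh pelvis transform and sets blend alphas, gravity and simulation space from sprint, movement, skydive, parachute and front-end states.
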AnimBlueprint /BRCosmetics/Accessories/FORT_Backpacks/Backpack_F_MED_RoseJewel/Meshes/F_MED_RoseJewel_Pack_AnimBP
target_skeleton=&/F_MED_RoseJewel_Pack_Skeleton
graphs [
 AnimGraph {nodes [
   n1 = RigidBody!(
    ComponentPose=Local To Component(LocalPose=Use cached pose 'Base pose '())
    Alpha=Get Sprint_DYN() -> (Sprint_DYN:double)
    OverrideWorldGravity:Vector="0,0,-980"
    ExternalForce
   ) -> (Pose) {
    bOverrideWorldGravity=true
    overridePhysicsAsset=&/F_MED_RoseJewel_Pack_Physics
    simulationSpace=BaseBoneSpace
    alphaBoolBlend {blendOption=Linear}
    baseBoneRef {boneName=pack}
    componentAppliedLinearAccClamp=<1000 1000 1000>
    overrideWorldGravity {z=-980}
    simSpaceSettings {maxAngularAcceleration=40, maxAngularVelocity=80, maxLinearAcceleration=1500, maxLinearVelocity=800, worldAlpha=0.65}
   }
   n2 = "Transform (Modify) Bone - Bone: base_strap"(ComponentPose=Local To Component(LocalPose=Use cached pose 'Base pose '()), Alpha=1, Translation="0,0,0", Rotation="0,0,15", Scale="1,1,1") {rotationMode=BMM_Additive, scaleMode=BMM_Additive, translationMode=BMM_Additive, boneToModify {boneName=base_strap}, rotation {roll=15}}
   n3 = Anim Dynamics - Bone: dyn_strap_2!(SimulationSpace:AnimPhysSimSpaceType Enum, ComponentPose=@n2, Alpha=1, ExternalForce="0,0,0") -> (Pose) {
    angularDampingOverride=1
    angularSpringConstant=12000
    bAngularSpring=true
    bOverrideAngularDamping=true
    bUseGravityOverride=true
    gravityScale=0.35
    simulationSpace=RootRelative
    boundBone {boneName=dyn_strap_2}
    componentAppliedLinearAccClamp=<2000 2000 2000>
    componentLinearAccScale=<1 1 1>
    componentLinearVelScale=<1 1 1>
    gravityOverride {z=-980}
    physicsBodyDefinitions [{
      boundBone {boneName=dyn_strap_2}
      boxExtents=<20 30 30>
      constraintSetup {angularLimitsMax=<0 5 5>, angularLimitsMin=<0 -5 -5>, angularTarget=<1 0 0>}
      localJointOffset=<4 0 0>
     }]
   }
   n4 = Anim Dynamics - Bone: dyn_strap_2(ComponentPose=@n3, Alpha=1, ExternalForce="0,0,0") {
    angularDampingOverride=1
    angularSpringConstant=12000
    bAngularSpring=true
    bOverrideAngularDamping=true
    bUseGravityOverride=true
    bUseSphericalLimits=true
    gravityScale=0.35
    simulationSpace=RootRelative
    boundBone {boneName=dyn_strap_2}
    componentAppliedLinearAccClamp=<2000 2000 2000>
    componentLinearAccScale=<1 1 1>
    componentLinearVelScale=<1 1 1>
    gravityOverride {z=-980}
    physicsBodyDefinitions [{
      boundBone {boneName=dyn_strap_2}
      boxExtents=<20 30 30>
      constraintSetup {angularLimitsMax=<0 5 15>, angularLimitsMin=<0 -5 0>, angularTarget=<1 0 0>}
      localJointOffset=<4 0 0>
     }]
    sphericalLimits [{limitRadius=15, drivingBone {boneName=Hip_Collision}, sphereLocalOffset=<0 -5 5>}]
   }
   n5 = Anim Dynamics - Bone: dyn_strap_3!(SimulationSpace:AnimPhysSimSpaceType Enum, ComponentPose=@n4, Alpha=1, ExternalForce="0,0,0") -> (Pose) {
    angularDampingOverride=1
    angularSpringConstant=12000
    bAngularSpring=true
    bOverrideAngularDamping=true
    bUseGravityOverride=true
    bUseSphericalLimits=true
    gravityScale=0.35
    simulationSpace=RootRelative
    boundBone {boneName=dyn_strap_3}
    componentAppliedLinearAccClamp=<2000 2000 2000>
    componentLinearAccScale=<1 1 1>
    componentLinearVelScale=<1 1 1>
    gravityOverride {z=-980}
    physicsBodyDefinitions [{
      boundBone {boneName=dyn_strap_3}
      boxExtents=<20 30 30>
      constraintSetup {angularLimitsMax=<0 5 15>, angularLimitsMin=<0 -5 0>, angularTarget=<1 0 0>}
      localJointOffset=<4 0 0>
     }]
    sphericalLimits [{limitRadius=15, drivingBone {boneName=Hip_Collision}, sphereLocalOffset=<0 -5 5>}]
   }
   n6 = Anim Dynamics - Bone: dyn_strap_4!(SimulationSpace:AnimPhysSimSpaceType Enum, ComponentPose=@n5, Alpha=1, ExternalForce="0,0,0") -> (Pose) {
    angularDampingOverride=1
    angularSpringConstant=12000
    bAngularSpring=true
    bOverrideAngularDamping=true
    bUseGravityOverride=true
    bUseSphericalLimits=true
    gravityScale=0.35
    simulationSpace=RootRelative
    boundBone {boneName=dyn_strap_4}
    componentAppliedLinearAccClamp=<2000 2000 2000>
    componentLinearAccScale=<1 1 1>
    componentLinearVelScale=<1 1 1>
    gravityOverride {z=-980}
    physicsBodyDefinitions [{
      boundBone {boneName=dyn_strap_4}
      boxExtents=<20 30 30>
      constraintSetup {angularLimitsMax=<0 5 15>, angularLimitsMin=<0 -5 0>, angularTarget=<1 0 0>}
      localJointOffset=<4 0 0>
     }]
    sphericalLimits [{limitRadius=15, drivingBone {boneName=Hip_Collision}, sphereLocalOffset=<0 -5 5>}]
   }
   n7 = Trail controller - Bone: dyn_strap_5(ComponentPose=@n6, Alpha=Get MovingAlpha() -> (MovingAlpha:double)) {
    bLimitRotation=true
    chainLength=4
    maxDeltaTime=0.333
    relaxationSpeedScale=2.4
    alphaScaleBiasClamp {bInterpResult=true, interpSpeedDecreasing=5, interpSpeedIncreasing=5}
    relaxationSpeedScaleInputProcessor {interpSpeedDecreasing=50, interpSpeedIncreasing=50}
    rotationLimits [
     {limitMax=<20 20 20>, limitMin=<-20 -20 -20>}
     {limitMax=<40 40 40>, limitMin=<-20 -20 -60>}
     {limitMax=<60 60 60>, limitMin=<-20 -60 -60>}
     {limitMax=<180 180 180>}
    ]
    rotationOffsets [<0 0 0>, <0 0 0>, <0 0 0>, <0 0 0>]
    trailBone {boneName=dyn_strap_5}
   }
   n8 = Blend Poses by bool(BlendPose_0=Component To Local(ComponentPose=@n1), BlendPose_1=Component To Local(ComponentPose=@n7), bActiveValue=Get bRigidBodyEnabled()) {blendType=Linear}
   Output Pose(Result=Component To Local(ComponentPose=Local To Component(LocalPose=@n8))) {layerGroup=None}
   n9 = "Transform (Modify) Bone - Bone: pack"(ComponentPose=Local To Component(LocalPose=Input Pose()), Translation="0,1,0", Rotation="0,0,0") {
    rotationMode=BMM_Additive
    rotationSpace=BCS_BoneSpace
    translationMode=BMM_Additive
    translationSpace=BCS_BoneSpace
    boneToModify {boneName=pack}
    translation {y=1}
   }
   n10 = "Transform (Modify) Bone - Bone: Hip_Collision"!(ComponentPose=@n9, bAlphaBoolEnabled:bool=True, Translation="0,0,0", Rotation="0,0,0", Scale="1,1,1") -> (Pose) {
    alphaInputType=Bool
    rotationMode=BMM_Replace
    rotationSpace=BCS_WorldSpace
    translationMode=BMM_Replace
    translationSpace=BCS_WorldSpace
    boneToModify {boneName=Hip_Collision}
   }
   n11 = "Transform (Modify) Bone - Bone: pack"!(ComponentPose=@n10, bAlphaBoolEnabled:bool, Translation="0,0,0", Rotation="0,0,5", Scale) -> (Pose) {
    alphaInputType=Bool
    rotationMode=BMM_Additive
    rotationSpace=BCS_BoneSpace
    translationMode=BMM_Additive
    translationSpace=BCS_BoneSpace
    boneToModify {boneName=pack}
    rotation {roll=5}
   }
   n12 = "Transform (Modify) Bone - Bone: main_base"(
    ComponentPose=@n11
    Alpha=Get parachute_olnyopen() -> (parachute_olnyopen:double)
    Translation="0.000,-4.768,-4.301"
    Rotation="-0.000,0.000,31.718"
    Scale="1,1,1"
   ) {
    rotationMode=BMM_Additive
    rotationSpace=BCS_BoneSpace
    translationMode=BMM_Additive
    translationSpace=BCS_BoneSpace
    boneToModify {boneName=main_base}
    rotation {roll=31.718}
    translation {y=-4.768, z=-4.301}
   }
   n13 = "Transform (Modify) Bone - Bone: base_strap"!(ComponentPose=@n12, bAlphaBoolEnabled:bool=True, Translation="0,0,0", Rotation="0,20,0", Scale="1,1,1") -> (Pose) {
    alphaInputType=Bool
    rotationMode=BMM_Additive
    rotationSpace=BCS_BoneSpace
    translationMode=BMM_Additive
    translationSpace=BCS_BoneSpace
    boneToModify {boneName=base_strap}
    rotation {yaw=20}
   }
   n14 = Anim Dynamics - Bone: dyn_rose_b(ComponentPose=@n13, Alpha=1, ExternalForce="0,0,0") {
    angularDampingOverride=1
    angularSpringConstant=20000
    bAngularSpring=true
    bOverrideAngularDamping=true
    bUseGravityOverride=true
    gravityScale=0.35
    simulationSpace=RootRelative
    boundBone {boneName=dyn_rose_b}
    componentAppliedLinearAccClamp=<2000 2000 2000>
    componentLinearAccScale=<1 1 1>
    componentLinearVelScale=<1 1 1>
    gravityOverride {z=-980}
    physicsBodyDefinitions [{
      boundBone {boneName=dyn_rose_b}
      boxExtents=<15 15 15>
      constraintSetup {angularLimitsMax=<0 5 5>, angularLimitsMin=<0 -5 -5>, angularTarget=<1 0 0>}
      localJointOffset=<4 0 0>
     }]
   }
   n15 = Anim Dynamics - Bone: dyn_rose_c(ComponentPose=@n14, Alpha=1, ExternalForce="0,0,0") {
    angularDampingOverride=1
    angularSpringConstant=20000
    bAngularSpring=true
    bOverrideAngularDamping=true
    bUseGravityOverride=true
    gravityScale=0.35
    simulationSpace=RootRelative
    boundBone {boneName=dyn_rose_c}
    componentAppliedLinearAccClamp=<2000 2000 2000>
    componentLinearAccScale=<1 1 1>
    componentLinearVelScale=<1 1 1>
    gravityOverride {z=-980}
    physicsBodyDefinitions [{
      boundBone {boneName=dyn_rose_c}
      boxExtents=<15 15 15>
      constraintSetup {angularLimitsMax=<0 5 5>, angularLimitsMin=<0 -5 -5>, angularTarget=<1 0 0>}
      localJointOffset=<4 0 0>
     }]
   }
   n16 = Anim Dynamics - Bone: dyn_rose_d(ComponentPose=@n15, Alpha=1, ExternalForce="0,0,0") {
    angularDampingOverride=1
    angularSpringConstant=20000
    bAngularSpring=true
    bOverrideAngularDamping=true
    bUseGravityOverride=true
    gravityScale=0.35
    simulationSpace=RootRelative
    boundBone {boneName=dyn_rose_d}
    componentAppliedLinearAccClamp=<2000 2000 2000>
    componentLinearAccScale=<1 1 1>
    componentLinearVelScale=<1 1 1>
    gravityOverride {z=-980}
    physicsBodyDefinitions [{
      boundBone {boneName=dyn_rose_d}
      boxExtents=<15 15 15>
      constraintSetup {angularLimitsMax=<0 5 5>, angularLimitsMin=<0 -5 -5>, angularTarget=<1 0 0>}
      localJointOffset=<4 0 0>
     }]
   }
   n17 = Anim Dynamics - Bone: dyn_rose_e(ComponentPose=@n16, Alpha=1, ExternalForce="0,0,0") {
    angularDampingOverride=1
    angularSpringConstant=20000
    bAngularSpring=true
    bOverrideAngularDamping=true
    bUseGravityOverride=true
    gravityScale=0.35
    simulationSpace=RootRelative
    boundBone {boneName=dyn_rose_e}
    componentAppliedLinearAccClamp=<2000 2000 2000>
    componentLinearAccScale=<1 1 1>
    componentLinearVelScale=<1 1 1>
    gravityOverride {z=-980}
    physicsBodyDefinitions [{
      boundBone {boneName=dyn_rose_e}
      boxExtents=<15 15 15>
      constraintSetup {angularLimitsMax=<0 4 4>, angularLimitsMin=<0 -4 -4>, angularTarget=<1 0 0>}
      localJointOffset=<4 0 0>
     }]
   }
   n18 = Anim Dynamics - Bone: dyn_rose_f(ComponentPose=@n17, Alpha=1, ExternalForce="0,0,0") {
    angularDampingOverride=1
    angularSpringConstant=20000
    bAngularSpring=true
    bOverrideAngularDamping=true
    bUseGravityOverride=true
    gravityScale=0.35
    simulationSpace=RootRelative
    boundBone {boneName=dyn_rose_f}
    componentAppliedLinearAccClamp=<2000 2000 2000>
    componentLinearAccScale=<1 1 1>
    componentLinearVelScale=<1 1 1>
    gravityOverride {z=-980}
    physicsBodyDefinitions [{
      boundBone {boneName=dyn_rose_f}
      boxExtents=<15 15 15>
      constraintSetup {angularLimitsMax=<0 0 4>, angularLimitsMin=<0 -4 -4>, angularTarget=<1 0 0>}
      localJointOffset=<4 0 0>
     }]
   }
   n19 = Anim Dynamics - Bone: dyn_rose_g(ComponentPose=@n18, Alpha=1, ExternalForce="0,0,0") {
    angularDampingOverride=1
    angularSpringConstant=20000
    bAngularSpring=true
    bOverrideAngularDamping=true
    bUseGravityOverride=true
    gravityScale=0.35
    simulationSpace=RootRelative
    boundBone {boneName=dyn_rose_g}
    componentAppliedLinearAccClamp=<2000 2000 2000>
    componentLinearAccScale=<1 1 1>
    componentLinearVelScale=<1 1 1>
    gravityOverride {z=-980}
    physicsBodyDefinitions [{
      boundBone {boneName=dyn_rose_g}
      boxExtents=<15 15 15>
      constraintSetup {angularLimitsMax=<0 3 3>, angularLimitsMin=<0 -3 -3>, angularTarget=<1 0 0>}
      localJointOffset=<4 0 0>
     }]
   }
   n20 = Anim Dynamics - Bone: dyn_rose_a(ComponentPose=@n19, Alpha=1, ExternalForce="0,0,0") {
    angularDampingOverride=1
    angularSpringConstant=20000
    bAngularSpring=true
    bOverrideAngularDamping=true
    bUseGravityOverride=true
    gravityScale=0.35
    simulationSpace=RootRelative
    boundBone {boneName=dyn_rose_a}
    componentAppliedLinearAccClamp=<2000 2000 2000>
    componentLinearAccScale=<1 1 1>
    componentLinearVelScale=<1 1 1>
    gravityOverride {z=-980}
    physicsBodyDefinitions [{
      boundBone {boneName=dyn_rose_a}
      boxExtents=<15 15 15>
      constraintSetup {angularLimitsMax=<0 2 2>, angularLimitsMin=<0 -2 -2>, angularTarget=<1 0 0>}
      localJointOffset=<4 0 0>
     }]
   }
   Save cached pose 'Base pose '(Pose=Component To Local(ComponentPose=@n20)) {cacheName="Base pose "}
  ]}
 EventGraph {nodes [
   Event BlueprintUpdateAnimation()
   > n1 = Sequence!(execute) -> (then_0, then_1, then_2:Exec, then_3:Exec, then_4:Exec)
   n2 = NOT Boolean(A=Get bParachuteOpenedAndStandingUpright() -> (bParachuteOpenedAndStandingUpright:bool))
   n3 = Get bIsParachuteOpened()
   > Set parachute_olnyopen(parachute_olnyopen:double="To Float (Boolean)"(InBool=AND Boolean(A=@n3, B=@n2))) -> (Output_Get:double)
   > n4 = Branch(Condition=OR Boolean!(
     A=Get bIsSkydiveFloating()
     B=Get bIsSkydiveDiveMode()
     C:bool=Get bIsKneeSliding()
     D:bool=Get bIsBaseballSliding()
     E:bool=Get bIsParachuteOpened()
    ) -> (ReturnValue))
   > Set Gravity_Override(Gravity_Override:Vector="0,0,450") -> (Output_Get:Vector)
   Set Gravity_Override(execute=@n4.else, Gravity_Override="0,0,-980")
   n5 = GetPartSkeletalMeshComponent()
   Is Valid(exec=@n1.then_1, InputObject=@n5)
   n6 = GetAttachParent!(self:Scene Component Object Reference=@n5) -> (ReturnValue:Scene Component Object Reference)
   > Is Valid(InputObject=@n6)
   > n7 = Set Body_Mesh(Body_Mesh:Scene Component Object Reference=@n6) -> (Output_Get:Scene Component Object Reference)
   n8 = BreakTransform(InTransform=GetSocketTransform(self=@n7, InSocketName=pelvis, TransformSpace=RTS_World))
   > Set pelvis_location(pelvis_location:Vector=@n8) -> (Output_Get:Vector)
   > Set pelvis_rotation(pelvis_rotation:Rotator=@n8.Rotation) -> (Output_Get:Rotator)
   n9 = Branch(
    execute=@n1.then_3
    Condition=OR Boolean!(
     A=Get bIsSprinting()
     B=Get bIsMoving2D()
     C=Get bIsJumping()
     D=Get bIsFalling()
     E=Get bIsSkyDiving()
     F:bool=Get bIsInFrontEnd()
     G:bool=Get bIsPlayingEmote()
    ) -> (ReturnValue)
   )
   > Set MovingAlpha(MovingAlpha:double=0.3) -> (Output_Get:double)
   Set MovingAlpha(execute=@n9.else, MovingAlpha=1)
   n10 = Branch(execute=@n1.then_4, Condition=Get bIsInFrontEnd())
   > Set Sim Space(Sim Space:AnimPhysSimSpaceType Enum=World) -> (Output_Get:AnimPhysSimSpaceType Enum)
   > Set DynAlpha(DynAlpha:double=0.3) -> (Output_Get:double)
   Set Sim Space(execute=@n10.else)
   > Set DynAlpha(DynAlpha=1)
   n11 = Branch(execute=@n1.then_2, Condition=OR Boolean!(A=Get bIsSprinting(), B=Get bIsTacticalSprinting()) -> (ReturnValue))
   > Set Sprint_DYN(Sprint_DYN:double=0.7) -> (Output_Get:double)
   Set Sprint_DYN(execute=@n11.else, Sprint_DYN=1)
   TryGetPawnOwner()
   Get bIsSprinting()
   Set Gravity_Override(Gravity_Override="0,0,450")
  ]}
]
variables [
 Body_Mesh {type=SceneComponent, default=null}
 DynAlpha {type=number}
 Gravity_Override {type=Vector}
 MovingAlpha {type=number}
 parachute_olnyopen {type=number}
 pelvis_location {type=Vector}
 pelvis_rotation {type=Rotator}
 Sim Space {type=AnimPhysSimSpaceType}
 Skeletal Mesh {type=SkeletalMeshComponent}
 Skydive_Float {type=number}
 Sprint_DYN {type=number}
]
